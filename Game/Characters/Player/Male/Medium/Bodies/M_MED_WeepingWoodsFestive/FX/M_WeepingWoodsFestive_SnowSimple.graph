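# Simple lit snow material for Niagara mesh particles in the Weeping Woods Festive character FX.
# A Fresnel blends two color parameters, white to dark blue by default, and a snowball normal map is softened toward flat.
# Tiled Perlin noise drives a scalar blend.
# A second noise sample, offset by a per-particle random value, scales the world vertex normal, forming a displacement-style vector that varies per particle.
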Material /Game/Characters/Player/Male/Medium/Bodies/M_MED_WeepingWoodsFestive/FX/M_WeepingWoodsFestive_SnowSimple
graph {
 properties {bUsedWithNiagaraMeshParticles=true, pixelDepthOffsetMode=PDOM_Legacy, usedShadingModels=Default Lit}
 nodes [
  LinearInterpolate(
   A=VectorParameter() {parameterName=Color1, defaultValue=<rgba 1 1 1 0>}.RGB
   B=VectorParameter() {parameterName=Color2, defaultValue=<rgba 0.051 0.134 0.3 1>}.RGB
   Alpha=Fresnel() {baseReflectFraction=0.2, exponent=1}
  )
  LinearInterpolate(
   A=TextureSampleParameter2D() {parameterName=Normal, samplerType=SAMPLERTYPE_Normal, texture=/Game/Abilities/Toys/Snowball/Textures/T_Snowball_Normals}.RGB
   B=Constant3Vector() {constant=<rgba 0 0 1 0>}
  )
  n1 = TextureSampleParameter2D(UVs=Multiply(A=TextureCoordinate()) {constB=8}) {
   parameterName=Noise
   samplerType=SAMPLERTYPE_LinearGrayscale
   texture=&T_SimplePerlin_Greyscale
  }
  LinearInterpolate(Alpha=Power(Base=@n1.RGB)) {constA=0.5, constB=0}
  n2 = TextureCoordinate()
  n3 = TextureSampleParameter2D(UVs=Add(A=Multiply(A=@n2), B=ParticleRandom())) {
   parameterName=Noise
   samplerType=SAMPLERTYPE_LinearGrayscale
   texture=&T_SimplePerlin_Greyscale
  }
  Multiply(A=Multiply(A=VertexNormalWS(), B=@n3.G)) {constB=3}
 ]
}
